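# Translucent unlit material named as a temporary weapon material in the lightning item's idle FX, usable on skeletal meshes and mesh deformers.
# It blends two blue HDR color parameters along the V texture coordinate, with the falloff shaped by a power exponent parameter.
# A much brighter blue Fresnel rim term is added on top of that gradient.
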Material /SunRoseZeusGameplay/Items/Lightning/FX/Idle/M_TempWeaponMat
graph {
 properties {
  bUsedWithMeshDeformer=true
  bUsedWithSkeletalMesh=true
  blendMode=BLEND_Translucent
  pixelDepthOffsetMode=PDOM_Legacy
  shadingModel=MSM_Unlit
  usedShadingModels=Unlit
 }
 nodes [
  n1 = MaterialFunctionCall(Float2=TextureCoordinate()) {materialFunction=&BreakOutFloat2Components}
  n2 = ScalarParameter() {defaultValue=4.5, parameterName=WET}
  Add(
   A=LinearInterpolate(
    A=VectorParameter() {parameterName=Color, defaultValue=<rgba 0.17 0.935 1.7 1>}.RGB
    B=VectorParameter() {parameterName=ColorB, defaultValue=<rgba 2.5 10.4166 50 1>}.RGB
    Alpha=Power(Base=@n1.G, Exp=@n2)
   )
   B=Multiply(A=VectorParameter() {parameterName=Color2, defaultValue=<rgba 15 62.4998 300 1>}.RGB, B=Fresnel() {baseReflectFraction=0})
  )
  Constant()
  OneMinus(None=@n1.G)
 ]
}
